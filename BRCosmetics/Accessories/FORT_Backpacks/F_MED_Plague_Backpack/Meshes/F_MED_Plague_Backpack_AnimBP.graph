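# Animation Blueprint for the Plague backpack cosmetic that adds secondary motion to its jug, vials, bell, syringe and their liquids.
# The input pose runs through a serial chain of per-bone Anim Dynamics nodes; two extra jug nodes sit unconnected.
# Each update, skydiving lowers jug and jug-stopper simulation alpha.
# Skydiving without an open parachute also applies an additive offset to the bell sling pivot.
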AnimBlueprint /BRCosmetics/Accessories/FORT_Backpacks/F_MED_Plague_Backpack/Meshes/F_MED_Plague_Backpack_AnimBP
target_skeleton=&/F_MED_Plague_Backpack_Skeleton
graphs [
 AnimGraph {nodes [
   Anim Dynamics - Bone: jug(Alpha=0.2, ExternalForce="0,0,0") {
    angularDampingOverride=0.999
    angularSpringConstant=500
    bAngularSpring=true
    bGravityOverrideInSimSpace=true
    bLinearSpring=true
    bOverrideAngularDamping=true
    bOverrideLinearDamping=true
    gravityScale=0.6
    linearDampingOverride=0.999
    linearSpringConstant=50
    alphaBoolBlend {blendOption=Linear}
    boundBone {boneName=jug}
    componentAppliedLinearAccClamp=<3000 3000 3000>
    componentLinearAccScale=<1 1 1>
    componentLinearVelScale=<1 1 1>
    physicsBodyDefinitions [{
      sphereCollisionRadius=0
      boundBone {boneName=jug}
      boxExtents=<45 45 22>
      constraintSetup {angularLimitsMax=<0 0 0>, angularLimitsMin=<0 0 0>, angularTarget=<0 0 0>}
      localJointOffset=<2 0 -26>
     }]
   }
   Anim Dynamics - Bone: jug(Alpha=1, ExternalForce="0,0,0") {
    angularDampingOverride=0.999
    angularSpringConstant=500
    bAngularSpring=true
    bGravityOverrideInSimSpace=true
    bLinearSpring=true
    bOverrideAngularDamping=true
    bOverrideLinearDamping=true
    gravityScale=0.6
    linearDampingOverride=0.999
    linearSpringConstant=50
    alphaBoolBlend {blendOption=Linear}
    boundBone {boneName=jug}
    componentAppliedLinearAccClamp=<3000 3000 3000>
    componentLinearAccScale=<1 1 1>
    componentLinearVelScale=<1 1 1>
    physicsBodyDefinitions [{
      sphereCollisionRadius=0
      boundBone {boneName=jug}
      boxExtents=<45 45 22>
      constraintSetup {angularLimitsMax=<15 15 0>, angularLimitsMin=<-15 -15 -44>, angularTarget=<0 0 0>, linearAxesMax=<0 0 12>}
      localJointOffset=<2 0 -26>
     }]
   }
   AnimGraphNode_SubInput_0 = Input Pose()
   n1 = Get IsNotInAir() -> (IsNotInAir:double)
   n2 = Anim Dynamics - Bone: jug(ComponentPose=Local To Component(LocalPose=@AnimGraphNode_SubInput_0), Alpha=@n1, ExternalForce="0,0,0") {
    angularDampingOverride=0.999
    angularSpringConstant=500
    bAngularSpring=true
    bGravityOverrideInSimSpace=true
    bLinearSpring=true
    bOverrideAngularDamping=true
    bOverrideLinearDamping=true
    gravityScale=0.6
    linearDampingOverride=0.999
    linearSpringConstant=50
    alphaBoolBlend {blendOption=Linear}
    boundBone {boneName=jug}
    componentAppliedLinearAccClamp=<1000 1000 1000>
    componentLinearAccScale=<1 1 1>
    componentLinearVelScale=<1 1 1>
    physicsBodyDefinitions [{
      sphereCollisionRadius=0
      boundBone {boneName=jug}
      boxExtents=<45 45 22>
      constraintSetup {angularLimitsMax=<15 2 0>, angularLimitsMin=<-15 -2 -44>, angularTarget=<0 0 0>}
      localJointOffset=<2 0 -26>
     }]
   }
   n3 = Anim Dynamics - Bone: dyn_jugJuice(ComponentPose=@n2, Alpha=1, ExternalForce="0,0,0") {
    angularDampingOverride=0.999
    angularSpringConstant=1000
    bAngularSpring=true
    bGravityOverrideInSimSpace=true
    bOverrideAngularDamping=true
    alphaBoolBlend {blendOption=Linear}
    boundBone {boneName=dyn_jugJuice}
    componentAppliedLinearAccClamp=<50 50 50>
    componentLinearAccScale=<1 1 1>
    componentLinearVelScale=<1 1 1>
    physicsBodyDefinitions [{
      sphereCollisionRadius=0
      boundBone {boneName=dyn_jugJuice}
      boxExtents=<18 22 8>
      constraintSetup {angularLimitsMax=<30 30 30>, angularLimitsMin=<-30 -30 -30>, angularTarget=<0 0 0>}
      localJointOffset=<0 -0.05 -11>
     }]
   }
   n4 = Anim Dynamics - Bone: dyn_jugstopper(ComponentPose=@n3, Alpha=@n1, ExternalForce="0,0,0") {
    angularDampingOverride=0.999
    angularSpringConstant=500
    bAngularSpring=true
    bGravityOverrideInSimSpace=true
    bLinearSpring=true
    bOverrideAngularDamping=true
    gravityScale=0.1
    linearSpringConstant=500
    alphaBoolBlend {blendOption=Linear}
    boundBone {boneName=dyn_jugstopper}
    componentAppliedLinearAccClamp=<3000 3000 3000>
    componentLinearAccScale=<1 1 1>
    componentLinearVelScale=<1 1 1>
    physicsBodyDefinitions [{
      sphereCollisionRadius=0
      boundBone {boneName=dyn_jugstopper}
      boxExtents=<2 2 2>
      constraintSetup {angularLimitsMax=<2 2 1>, angularLimitsMin=<-2 -2 -1>, angularTarget=<0 0 0>}
      localJointOffset=<-0.1 0 -0.1>
     }]
   }
   n5 = Anim Dynamics - Bone: dyn_vial_l(ComponentPose=@n4, Alpha=1, ExternalForce="0,0,0") {
    angularDampingOverride=0.999
    angularSpringConstant=1000
    bAngularSpring=true
    bGravityOverrideInSimSpace=true
    bOverrideAngularDamping=true
    linearSpringConstant=1000
    alphaBoolBlend {blendOption=Linear}
    boundBone {boneName=dyn_vial_l}
    componentAppliedLinearAccClamp=<1000 1000 1000>
    componentLinearAccScale=<1 1 1>
    componentLinearVelScale=<1 1 1>
    physicsBodyDefinitions [{
      sphereCollisionRadius=0
      boundBone {boneName=dyn_vial_l}
      boxExtents=<14 16 4>
      constraintSetup {angularTargetAxis=AxisY, twistAxis=AxisY, angularLimitsMax=<12 12 0>, angularLimitsMin=<-12 -12 0>, angularTarget=<0 0 0>}
      localJointOffset=<0 -0.05 -8>
     }]
   }
   n6 = Anim Dynamics - Bone: dyn_vialJucie_l(ComponentPose=@n5, Alpha=1, ExternalForce="0,0,0") {
    angularDampingOverride=0.95
    angularSpringConstant=1000
    bGravityOverrideInSimSpace=true
    bOverrideAngularDamping=true
    alphaBoolBlend {blendOption=Linear}
    boundBone {boneName=dyn_vialJucie_l}
    componentAppliedLinearAccClamp=<3000 3000 3000>
    componentLinearAccScale=<1 1 1>
    componentLinearVelScale=<1 1 1>
    physicsBodyDefinitions [{
      sphereCollisionRadius=0
      boundBone {boneName=dyn_vialJucie_l}
      boxExtents=<2 2 2>
      constraintSetup {
       angularTargetAxis=AxisY
       twistAxis=AxisY
       angularLimitsMax=<33 33 0>
       angularLimitsMin=<-33 -33 0>
       angularTarget=<0 0 0>
       linearAxesMax=<0 0 1>
      }
      localJointOffset=<0 0 -1>
     }]
   }
   n7 = Anim Dynamics - Bone: dyn_vial_c(ComponentPose=@n6, Alpha=1, ExternalForce="0,0,0") {
    angularDampingOverride=0.999
    angularSpringConstant=1000
    bAngularSpring=true
    bGravityOverrideInSimSpace=true
    bOverrideAngularDamping=true
    alphaBoolBlend {blendOption=Linear}
    boundBone {boneName=dyn_vial_c}
    componentAppliedLinearAccClamp=<3000 3000 3000>
    componentLinearAccScale=<1 1 1>
    componentLinearVelScale=<1 1 1>
    physicsBodyDefinitions [{
      sphereCollisionRadius=0
      boundBone {boneName=dyn_vial_c}
      boxExtents=<14 30 3>
      constraintSetup {angularTargetAxis=AxisY, twistAxis=AxisY, angularLimitsMax=<12 12 0>, angularLimitsMin=<-12 -12 0>, angularTarget=<0 0 0>}
      localJointOffset=<0 -0.05 -7>
     }]
   }
   n8 = Anim Dynamics - Bone: dyn_vialJucie_c(ComponentPose=@n7, Alpha=1, ExternalForce="0,0,0") {
    angularDampingOverride=0.95
    angularSpringConstant=1000
    bGravityOverrideInSimSpace=true
    bOverrideAngularDamping=true
    alphaBoolBlend {blendOption=Linear}
    boundBone {boneName=dyn_vialJucie_c}
    componentAppliedLinearAccClamp=<3000 3000 3000>
    componentLinearAccScale=<1 1 1>
    componentLinearVelScale=<1 1 1>
    physicsBodyDefinitions [{
      sphereCollisionRadius=0
      boundBone {boneName=dyn_vialJucie_c}
      boxExtents=<2 2 2>
      constraintSetup {
       angularTargetAxis=AxisY
       twistAxis=AxisY
       angularLimitsMax=<33 33 0>
       angularLimitsMin=<-33 -33 0>
       angularTarget=<0 0 0>
       linearAxesMax=<0 0 1>
      }
      localJointOffset=<0 0 -1>
     }]
   }
   n9 = Anim Dynamics - Bone: dyn_vial_r(ComponentPose=@n8, Alpha=1, ExternalForce="0,0,0") {
    angularDampingOverride=0.999
    angularSpringConstant=1000
    bAngularSpring=true
    bGravityOverrideInSimSpace=true
    bOverrideAngularDamping=true
    alphaBoolBlend {blendOption=Linear}
    boundBone {boneName=dyn_vial_r}
    componentAppliedLinearAccClamp=<3000 3000 3000>
    componentLinearAccScale=<1 1 1>
    componentLinearVelScale=<1 1 1>
    physicsBodyDefinitions [{
      sphereCollisionRadius=0
      boundBone {boneName=dyn_vial_r}
      boxExtents=<18 22 8>
      constraintSetup {angularTargetAxis=AxisY, twistAxis=AxisY, angularLimitsMax=<12 12 0>, angularLimitsMin=<-12 -12 0>, angularTarget=<0 0 0>}
      localJointOffset=<0 -0.05 -11>
     }]
   }
   n10 = Anim Dynamics - Bone: dyn_vialJucie_r(ComponentPose=@n9, Alpha=1, ExternalForce="0,0,0") {
    angularDampingOverride=0.95
    angularSpringConstant=1000
    bGravityOverrideInSimSpace=true
    bOverrideAngularDamping=true
    alphaBoolBlend {blendOption=Linear}
    boundBone {boneName=dyn_vialJucie_r}
    componentAppliedLinearAccClamp=<3000 3000 3000>
    componentLinearAccScale=<1 1 1>
    componentLinearVelScale=<1 1 1>
    physicsBodyDefinitions [{
      sphereCollisionRadius=0
      boundBone {boneName=dyn_vialJucie_r}
      boxExtents=<2 2 2>
      constraintSetup {
       angularTargetAxis=AxisY
       twistAxis=AxisY
       angularLimitsMax=<33 33 0>
       angularLimitsMin=<-33 -33 0>
       angularTarget=<0 0 0>
       linearAxesMax=<0 0 1>
      }
      localJointOffset=<0 0 -1>
     }]
   }
   n11 = Anim Dynamics - Bone: dyn_bellslingpivot(ComponentPose=@n10, Alpha=1, ExternalForce="0,0,0") {
    angularDampingOverride=0.999
    angularSpringConstant=1500
    bAngularSpring=true
    bGravityOverrideInSimSpace=true
    bOverrideAngularDamping=true
    bOverrideLinearDamping=true
    linearDampingOverride=0.999
    alphaBoolBlend {blendOption=Linear}
    boundBone {boneName=dyn_bellslingpivot}
    componentAppliedLinearAccClamp=<1000 1000 1000>
    componentLinearAccScale=<1 1 1>
    componentLinearVelScale=<1 1 1>
    physicsBodyDefinitions [{
      sphereCollisionRadius=0
      boundBone {boneName=dyn_bellslingpivot}
      boxExtents=<8 22 26>
      constraintSetup {angularLimitsMax=<0 12 12>, angularLimitsMin=<0 -12 -12>, angularTarget=<0 0 0>, linearAxesMin=<-6 0 0>}
      localJointOffset=<20 0 -1>
     }]
   }
   n12 = "Transform (Modify) Bone - Bone: dyn_bellslingpivot"!(ComponentPose=@n11, Alpha=Get IsInAir() -> (IsInAir:double)) -> (Pose) {
    rotationMode=BMM_Additive
    translationMode=BMM_Additive
    alphaBoolBlend {blendOption=Linear}
    boneToModify {boneName=dyn_bellslingpivot}
    rotation {roll=-33}
    translation {y=-0.5, z=-1}
   }
   n13 = Anim Dynamics - Bone: dyn_bellclapper(ComponentPose=@n12, Alpha=0.5, ExternalForce="0,0,0") {
    angularDampingOverride=0.999
    angularSpringConstant=1000
    bAngularSpring=true
    bGravityOverrideInSimSpace=true
    bOverrideAngularDamping=true
    alphaBoolBlend {blendOption=Linear}
    boundBone {boneName=dyn_bellclapper}
    componentAppliedLinearAccClamp=<400 400 400>
    componentLinearAccScale=<1 1 1>
    componentLinearVelScale=<1 1 1>
    physicsBodyDefinitions [{
      sphereCollisionRadius=0
      boundBone {boneName=dyn_bellclapper}
      boxExtents=<8 22 26>
      constraintSetup {angularLimitsMax=<0 1 1>, angularLimitsMin=<0 -1 -1>, angularTarget=<0 0 0>}
      localJointOffset=<20 0 -1>
     }]
   }
   n14 = Anim Dynamics - Bone: dyn_syringestopper(ComponentPose=@n13, Alpha=1, ExternalForce="0,0,0") {
    angularDampingOverride=0.999
    angularSpringConstant=2000
    bAngularSpring=true
    bGravityOverrideInSimSpace=true
    bOverrideAngularDamping=true
    gravityScale=0.01
    alphaBoolBlend {blendOption=Linear}
    boundBone {boneName=dyn_syringestopper}
    componentAppliedLinearAccClamp=<200 200 200>
    componentLinearAccScale=<1 1 1>
    componentLinearVelScale=<1 1 1>
    physicsBodyDefinitions [{
      sphereCollisionRadius=0
      boundBone {boneName=dyn_syringestopper}
      boxExtents=<2 12 12>
      constraintSetup {angularTargetAxis=AxisY, twistAxis=AxisZ, angularLimitsMax=<1 3 3>, angularLimitsMin=<-1 -3 -3>, angularTarget=<0 0 0>}
      localJointOffset=<6 0 -4>
     }]
   }
   n15 = Anim Dynamics - Bone: dyn_syringeJuice(ComponentPose=@n14, Alpha=1, ExternalForce="0,0,0") {
    angularDampingOverride=0.95
    angularSpringConstant=2000
    bGravityOverrideInSimSpace=true
    bOverrideAngularDamping=true
    gravityScale=0.01
    alphaBoolBlend {blendOption=Linear}
    boundBone {boneName=dyn_syringeJuice}
    componentAppliedLinearAccClamp=<100 100 100>
    componentLinearAccScale=<1 1 1>
    componentLinearVelScale=<1 1 1>
    physicsBodyDefinitions [{
      sphereCollisionRadius=0
      boundBone {boneName=dyn_syringeJuice}
      boxExtents=<1 1 1>
      constraintSetup {angularTargetAxis=AxisY, angularLimitsMax=<33 33 0>, angularLimitsMin=<-33 -33 0>, angularTarget=<0 0 0>}
      localJointOffset=<0 0 -1>
     }]
   }
   Output Pose(Result=Component To Local(ComponentPose=@n15)) {layerGroup=None}
  ]}
 EventGraph {nodes [
   Event BlueprintUpdateAnimation()
   n1 = Get bIsSkyDiving()
   > Set IsNotInAir(IsNotInAir:double=SelectFloat(A=0.15, B=1, bPickA=@n1)) -> (Output_Get:double)
   > Set IsInAir(IsInAir:double=SelectFloat(A=0, B="To Float (Boolean)"(InBool=@n1), bPickA=Get bIsParachuteOpened())) -> (Output_Get:double)
   Set IsNotSkydiving(IsNotSkydiving:bool=false) -> (Output_Get:bool)
  ]}
]
variables [
 IsInAir {type=number}
 IsNotInAir {type=number}
 IsNotSkydiving {type=boolean}
]
